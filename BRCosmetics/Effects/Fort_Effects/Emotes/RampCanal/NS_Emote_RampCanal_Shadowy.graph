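# Niagara system for the Shadowy variant of the RampCanal emote, wrapping the character in dark smoke with blue mesh smoke and trails.
# Dark, smoke and smoke_detail sprite emitters attach to skeletal mesh bones or sockets chosen by particle partitioning; the smoke layers use SubUV animation and upward acceleration.
# Smoke001 and Trail mesh emitters spawn in box volumes, with their skeletal sampling disabled.
# A front-end preview flag adjusts Smoke001, while Trail adapts to time of day and platform.
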NiagaraSystem /BRCosmetics/Effects/Fort_Effects/Emotes/RampCanal/NS_Emote_RampCanal_Shadowy
dependencies {
 usedDataInterfaces [
  &NiagaraDataInterfaceCurve
  &NiagaraDataInterfaceColorCurve
  &NiagaraDataInterfacePlatformSet
  &NiagaraDataInterfaceSpriteRendererInfo
  &NiagaraDataInterfaceSkeletalMesh
 ]
 usedDynamicInputs []
 usedModules [
  &self:SystemSpawnScript.SystemScriptSource.SystemScriptGraph.NiagaraNodeAssignment_0.SetVariables_#1
  &SystemState
 ]
}
emitters {
 Dark {
  emitterUpdateScript [
   &EmitterState(Inactive Response=::NewEnumerator1, Life Cycle Mode=::NewEnumerator1, Loop Behavior=::NewEnumerator1, UseLoopDelay=false)
   &SpawnBurst_Instantaneous(Spawn Count=13, Use Spawn Probability=false, UseLoopCountLimit=false)
  ]
  particleSpawnScript [
   &PartitionParticles(Number Of Partitions=11)
   &InitializeParticle(
    Color Mode=::NewEnumerator0
    Lifetime=3.6
    Sprite Size Mode=::NewEnumerator1
    Uniform Sprite Size Max=85
    Uniform Sprite Size Min=55
    UsePositionOffset=false
    Write Material Random=true
   )
   &CameraOffset(Camera Offset Amount=10)
   &DynamicMaterialParameters(
    Index 0 Param 1=&RandomRangeFloat(Minimum=0.5, Maximum=0.36)
    Module.Index0Float0Write=true
    Module.Index0Float1Write=false
    Module.Index0Float2Write=false
    Module.Index0Float3Write=false
   )
  ]
  particleUpdateScript [
   &ParticleState()
   &SkeletalMeshLocation(
    Bone / Socket Index=$Particles.PartitionParticles.Partition
    Bone Sampling Mode=::NewEnumerator4
    MaskSpawnGroup=false
    Skeletal Mesh=$System.Skeletal Mesh Reader
   )
   &ScaleColor(
    Scale Alpha=&FloatFromCurve(FloatCurve=DataInterface() {propertyValues {Curve {keys [<at 0 0>,<at 0.1 1>,<at 0.8 1>,<at 1 0>]}}}, Scale Curve=0.8)
    Scale RGB=&MakeVectorFromLinearColorRGB(Color=<rgba 0 0 0 1>)
    ScaleA=true
    ScaleRGB=true
   )
  ]
  rendererClasses [&NiagaraSpriteRendererProperties]
  renderers [{rendererClass=&NiagaraSpriteRendererProperties}]
 }
 Smoke001 {
  emitterUpdateScript [
   &EmitterState(Inactive Response=::NewEnumerator1, Life Cycle Mode=::NewEnumerator1, Loop Behavior=::NewEnumerator1, Loop Delay=0.5, UseLoopDelay=false)
   &SpawnBurst_Instantaneous(Spawn Count=5, Use Spawn Probability=false, UseLoopCountLimit=true)
   &SpawnRate(
    Spawn Probability=&FloatFromCurve(FloatCurve=DataInterface() {propertyValues {Curve {keys [<at 0 1>,<at 2.2 1>,<at 2.7 0>]}}}, CurveIndex=$Emitter.Age)
    SpawnRate=10
    Use Spawn Probability=true
   )
  ]
  particleSpawnScript [
   &InitializeParticle(
    Lifetime=1.5
    Mesh Scale=<0.9 1 1>
    Mesh Scale Mode=::NewEnumerator4
    Position Offset=&MakeCustomVectorFromBool(Bool=$User.bIsFrontEndPreview, True Vector=<20 0 0>, False Vector=<0 0 0>)
    UsePositionOffset=true
    Write Material Random=true
   )
   &InitialMeshOrientation(
    Mesh Orientation Mode=::NewEnumerator2
    Rotation=&RandomRangeVector(Minimum=<0 0 0>, Maximum=<0 1 0>)
    Use Rotation Vector=true
   ) {module=InitialMeshOrientation001}
   &ShapeLocation(
    Box Size=<100 20 10>
    MaskSpawnGroup=false
    Override Seed=false
    Rotation Angle=1
    Rotation Mode=::NewEnumerator1
    Shape Primitive=::NewEnumerator2
    Surface Only=false
    UseOwnerScale=true
   )
   &SkeletalMeshLocation(
    Bone Sampling Mode=::NewEnumerator4
    MaskSpawnGroup=false
    Sampled Position Offset=&AddVectorToPosition(Position=<0 0 0>, Vector=$Particles.ShapeLocation.ShapeVector)
    Skeletal Mesh=$System.Skeletal Mesh Reader
   ) {enabled=false}
   &CameraOffset(Camera Offset Amount=&MakeCustomFloatFromBool(Bool=$User.bIsFrontEndPreview, True Float=10, False Float=5))
  ]
  particleUpdateScript [
   &ParticleState()
   &ScaleMeshSize(Scale Factor=&VectorFromFloat(Value=&FloatFromCurve(
      FloatCurve=DataInterface() {propertyValues {Curve {keys [<at 0 0.7>,<at 0.2 1.025>,<at 0.4 0.9>,<at 1 0.2>]}}}
      Scale Curve=&MakeCustomFloatFromBool(Bool=$User.bIsFrontEndPreview, True Float=0.4, False Float=&RandomRangeFloat(Minimum=0.6, Maximum=0.8))
     )))
   &ScaleColor(
    Scale Alpha=&FloatFromCurve(
     FloatCurve=DataInterface() {propertyValues {Curve {keys [
         {arriveTangent=5.611, interpMode=RCIM_Cubic, leaveTangent=5.611, tangentMode=RCTM_User, time=0, value=0}
         {arriveTangent=0.0117, interpMode=RCIM_Cubic, leaveTangent=0.0117, tangentMode=RCTM_User, time=0.3, value=1}
         <at 1 0 RCIM_Cubic>
        ]}}}
     Scale Curve=&MakeCustomFloatFromBool(Bool=$User.bIsFrontEndPreview, True Float=0.4, False Float=0.6)
    )
    Scale RGB=&MakeVectorFromLinearColorRGB(Color=<rgba 0 0.4565 1 1>)
    ScaleA=true
    ScaleRGB=true
   )
   &DynamicMaterialParameters(
    Index 0 Param 1=1
    Index 0 Param 3=1
    Index 0 Param 4=&FloatFromCurve(FloatCurve=DataInterface() {propertyValues {Curve {keys [{arriveTangent=-2.87, interpMode=RCIM_Cubic, leaveTangent=-2.87, tangentMode=RCTM_User, time=0.3, value=1}, <at 1 0 RCIM_Cubic>]}}})
    Module.Index0Param0Write=true
    Module.Index0Param1Write=true
    Module.Index0Param2Write=true
    Module.Index0Param3Write=true
   )
  ]
  rendererClasses [&NiagaraMeshRendererProperties]
  renderers [{rendererClass=&NiagaraMeshRendererProperties, rendererIndex=0}]
 }
 Trail {
  emitterUpdateScript [
   &EmitterState(Inactive Response=::NewEnumerator1, Life Cycle Mode=::NewEnumerator1, Loop Behavior=::NewEnumerator1, UseLoopDelay=false)
   &SpawnBurst_Instantaneous(Spawn Count=8, Use Spawn Probability=false, UseLoopCountLimit=false)
  ]
  particleSpawnScript [
   &InitializeParticle(
    Color Maximum=<rgba 1.5 1.5 1.5 1>
    Color Minimum=<rgba 0.02 0.02 0.02 1>
    Color Mode=::NewEnumerator2
    Lifetime=3.6
    Mesh Scale=&RandomRangeVector(Minimum=<0.8 0.8 0.6>, Maximum=<1 1 0.8>)
    Mesh Scale Mode=::NewEnumerator4
    UsePositionOffset=false
    Write Material Random=true
   )
   &CameraOffset(Camera Offset Amount=15)
   &InitialMeshOrientation(Mesh Orientation Mode=::NewEnumerator2, Rotation=<0 0.25 0>, Use Rotation Vector=true) {module=InitialMeshOrientation001}
   &InitialMeshOrientation(
    Mesh Orientation Mode=::NewEnumerator2
    Rotation=&RandomRangeVector(Minimum=<0 0 0>, Maximum=<1 0 0>)
    Use Rotation Vector=true
   )
   &ShapeLocation(
    Box Size=<100 10 5>
    MaskSpawnGroup=false
    Override Seed=false
    Rotation Angle=1
    Rotation Mode=::NewEnumerator1
    Shape Primitive=::NewEnumerator2
    Surface Only=false
    UseOwnerScale=true
   )
   &SkeletalMeshLocation(
    Bone Sampling Mode=::NewEnumerator4
    MaskSpawnGroup=false
    Orientation Sampling=::NewEnumerator1
    Sampled Position Offset=&AddVectorToPosition(Position=<0 0 0>, Vector=$Particles.ShapeLocation.ShapeVector)
    Skeletal Mesh=$System.Skeletal Mesh Reader
   ) {enabled=false}
  ]
  particleUpdateScript [
   &ParticleState()
   &ScaleMeshSize(Scale Factor=&MakeVector(
     X=1
     Y=1
     Z=&FloatFromCurve(FloatCurve=DataInterface() {propertyValues {Curve {keys [
          {arriveTangent=0.7581, interpMode=RCIM_Cubic, leaveTangent=0.7581, tangentMode=RCTM_User, time=0, value=0.5}
          {arriveTangent=-0.2247, interpMode=RCIM_Cubic, leaveTangent=-0.2247, tangentMode=RCTM_User, time=0.1, value=0.95}
          {arriveTangent=-0.4155, interpMode=RCIM_Cubic, leaveTangent=-0.4155, tangentMode=RCTM_User, time=0.3, value=0.6}
          <at 1 0.4 RCIM_Cubic>
         ]}}})
    ))
   &ScaleColor(
    Scale Alpha=&FloatFromCurve(
     FloatCurve=DataInterface() {propertyValues {Curve {keys [<at 0 0>,<at 0.2 1>,<at 0.3 0.8>,<at 0.7 0.8>,<at 1 0>]}}}
     Scale Curve=&Multiply_Float(A=1.6, B=&FNTimeOfDayScalar(Night Scale=1, Morning Scale=0.5, Evening Scale=1))
    )
    Scale RGB=&MakeVectorFromLinearColorRGB(Color=&ColorFromCurve(ColorCurve=DataInterface() {propertyValues {
        AlphaCurve {keys []}
        BlueCurve {keys [<at 0.0355 50>,<at 0.2063 20 RCIM_Cubic>]}
        GreenCurve {keys [<at 0.0355 20>,<at 0.2063 10 RCIM_Cubic>]}
        RedCurve {keys [<at 0.0355 0>,<at 0.2063 0 RCIM_Cubic>]}
       }}))
    ScaleA=true
    ScaleRGB=true
   )
   &DynamicMaterialParameters(
    Index 0 Param 1=&RandomRangeFloat(Minimum=0.5, Maximum=1)
    Index 0 Param 2=&RandomRangeFloat(Minimum=0.5, Maximum=1)
    Index 0 Param 3=&RandomRangeFloat(Minimum=20, Maximum=10)
    Index 0 Param 4=1
    Index 3 Param 1=&FloatFromCurve(FloatCurve=DataInterface() {propertyValues {Curve {keys [
         <at 0 -625.0001 RCIM_Cubic>
         {arriveTangent=712.5001, interpMode=RCIM_Cubic, leaveTangent=712.5001, time=1, value=0}
         {arriveTangent=1495.8616, interpMode=RCIM_Cubic, leaveTangent=1495.8616, tangentMode=RCTM_User, time=2, value=800}
        ]}}})
    Module.Index0Float0Write=true
    Module.Index0Float1Write=true
    Module.Index0Float2Write=true
    Module.Index0Float3Write=true
    Module.Index1Float0Write=true
    Module.Index1Float1Write=true
    Module.Index1Float2Write=true
    Module.Index1Float3Write=true
    Module.Index3Float0Write=true
    Module.Index3Float1Write=true
    Module.Index3Float2Write=true
    Module.Index3Float3Write=true
    Write Parameter Index 1=::NewEnumerator1
    Write Parameter Index 3=::NewEnumerator1
   )
   &ScaleColor(
    Scale RGB=&Lerp_Vector(
     A=<1 1 1>
     B=<0.3 0.3 0.3>
     Alpha=&FN_PickFloatDependOnPlatform(TargetPlatformActive=0, TargetPlatformNoActive=1)
    )
    ScaleA=true
    ScaleRGB=true
   ) {module=ScaleColor001}
  ]
  rendererClasses [&NiagaraMeshRendererProperties]
  renderers [{rendererClass=&NiagaraMeshRendererProperties, rendererIndex=0}]
 }
 smoke {
  emitterUpdateScript [
   &EmitterState(Inactive Response=::NewEnumerator1, Life Cycle Mode=::NewEnumerator1, Loop Behavior=::NewEnumerator1, UseLoopDelay=false)
   &SpawnBurst_Instantaneous(Spawn Count=5, Use Spawn Probability=false, UseLoopCountLimit=false)
   &SpawnRate(
    Spawn Probability=&FloatFromCurve(FloatCurve=DataInterface() {propertyValues {Curve {keys [<at 0 1>,<at 1 1>,<at 1.3 0>]}}}, CurveIndex=$Emitter.NormalizedLoopAge)
    SpawnRate=15
    Use Spawn Probability=true
   )
   &SpawnBurst_Instantaneous(Spawn Count=11, Spawn Time=3.8833, Use Spawn Probability=false, UseLoopCountLimit=false) {enabled=false, module=SpawnBurst_Instantaneous001}
  ]
  particleSpawnScript [
   &PartitionParticles(Number Of Partitions=11)
   &InitializeParticle(
    Color=<rgba 0 0 0 0.5>
    Lifetime Max=2.5
    Lifetime Min=2
    Lifetime Mode=::NewEnumerator1
    Sprite Rotation Mode=::NewEnumerator2
    Sprite Size Mode=::NewEnumerator1
    Uniform Sprite Size Max=100
    Uniform Sprite Size Min=60
    UsePositionOffset=false
    Write Material Random=true
   ) {module=InitializeParticle001}
   &SkeletalMeshLocation(
    Bone / Socket Index=$Particles.PartitionParticles.Partition
    Bone Sampling Mode=::NewEnumerator4
    MaskSpawnGroup=false
    Orientation Sampling=::NewEnumerator1
    Skeletal Mesh=$System.Skeletal Mesh Reader
   )
   &AddVelocity(Scale Added Velocity=<1 1 1.1>, Velocity=<0 0 100>) {enabled=false}
   &CameraOffset(Camera Offset Amount=-20) {enabled=false}
  ]
  particleUpdateScript [
   &ParticleState()
   &SpriteRotationRate(Rotation Rate=-50)
   &Drag(Drag=0.5, Use Linear Drag=true, Use Rotational Drag=false)
   &ScaleColor(
    Scale Alpha=&FloatFromCurve(
     FloatCurve=DataInterface() {propertyValues {Curve {keys [
         {arriveTangent=1.7188, interpMode=RCIM_Cubic, leaveTangent=1.7188, tangentMode=RCTM_User, time=0, value=0}
         {arriveTangent=0.0226, interpMode=RCIM_Cubic, leaveTangent=0.0226, tangentMode=RCTM_User, time=0.2, value=0.5}
         {arriveTangent=0.0192, interpMode=RCIM_Cubic, leaveTangent=-1.0474, tangentMode=RCTM_Break, time=0.4, value=0.5}
         {arriveTangent=-0.0134, interpMode=RCIM_Cubic, leaveTangent=-0.0134, tangentMode=RCTM_User, time=1, value=0}
        ]}}}
     Scale Curve=0.7
    )
    ScaleA=true
    ScaleRGB=true
   )
   &DynamicMaterialParameters(
    Index 0 Param 1=&FloatFromCurve(FloatCurve=DataInterface() {propertyValues {Curve {keys [<at 0.3 1 RCIM_Cubic>,<at 1 -0.1>]}}}, CurveIndex=$Particles.NormalizedLife)
    Index 0 Param 2=&FloatFromCurve(
     FloatCurve=DataInterface() {propertyValues {Curve {keys [<at 0 1 RCIM_Cubic>,<at 0.8 8.8 RCIM_Cubic>]}}}
     CurveIndex=$Particles.NormalizedLife
    )
    Index 0 Param 4=&FloatFromCurve(FloatCurve=DataInterface() {propertyValues {Curve {keys [<at 0.3 0>,<at 0.5 0.6>]}}}, CurveIndex=$Particles.NormalizedLife)
    Module.Param0WriteEnabled=true
    Module.Param1WriteEnabled=true
    Module.Param2WriteEnabled=true
    Module.Param3WriteEnabled=true
   ) {enabled=false}
   &SubUVAnimation(
    End Frame Range Override=40
    "Sprite Renderer "=DataInterface() {propertyValues {SpriteRenderer=&self:smoke_1.NiagaraSpriteRendererProperties_2}}
    UseEndFrame=true
    UseStartFrame=true
   )
   &AccelerationForce(Acceleration=<0 0 100>)
   &SolveForcesAndVelocity(Clamp Velocity=false, Limit Acceleration=false)
   &DynamicMaterialParameters(
    Index 0 Param 1=&FloatFromCurve(
     FloatCurve=DataInterface() {propertyValues {Curve {keys [
         {interpMode=RCIM_Cubic, tangentMode=RCTM_User, time=0, value=0}
         <at 0.2 0.55>
         {arriveTangent=-1.0223, interpMode=RCIM_Cubic, leaveTangent=-1.0223, tangentMode=RCTM_User, time=1, value=1}
        ]}}}
     Scale Curve=&RandomRangeFloat(Minimum=3, Maximum=3)
    )
    Module.Index0Param0Write=true
    Module.Index0Param1Write=true
    Module.Index0Param2Write=true
    Module.Index0Param3Write=true
   ) {module=DynamicMaterialParameters001}
  ]
  rendererClasses [&NiagaraSpriteRendererProperties]
  renderers [{rendererClass=&NiagaraSpriteRendererProperties}]
 }
 smoke_detail {
  emitterUpdateScript [
   &EmitterState(Inactive Response=::NewEnumerator1, Life Cycle Mode=::NewEnumerator1, Loop Behavior=::NewEnumerator1, UseLoopDelay=false)
   &SpawnRate(SpawnRate=4, Use Spawn Probability=true)
  ]
  particleSpawnScript [
   &PartitionParticles(Number Of Partitions=11)
   &InitializeParticle(
    Color=<rgba 0 0 0 1>
    Lifetime=1.5
    Sprite Rotation Mode=::NewEnumerator2
    Sprite Size Mode=::NewEnumerator1
    Uniform Sprite Size Max=90
    Uniform Sprite Size Min=60
    UsePositionOffset=false
    Write Material Random=true
   ) {module=InitializeParticle001}
   &ShapeLocation(MaskSpawnGroup=false, Override Seed=false, Sphere Radius=50, UseOwnerScale=true)
   &AddVelocity(Module.Use Falloff Distance=false, Override Seed=false, Velocity Mode=::NewEnumerator1)
   &SkeletalMeshLocation(
    Bone / Socket Index=$Particles.PartitionParticles.Partition
    Bone Sampling Mode=::NewEnumerator4
    MaskSpawnGroup=false
    Orientation Sampling=::NewEnumerator1
    Sampled Position Offset=&AddVectorToPosition(Position=<0 0 0>, Vector=$Particles.ShapeLocation.ShapeVector)
    Skeletal Mesh=$System.Skeletal Mesh Reader
   )
  ]
  particleUpdateScript [
   &ParticleState()
   &SpriteRotationRate(Rotation Rate=&RandomRangeFloat(Minimum=-50, Maximum=50))
   &Drag(Drag=0.5, Use Linear Drag=true, Use Rotational Drag=false)
   &ScaleColor(
    Scale Alpha=&FloatFromCurve(
     FloatCurve=DataInterface() {propertyValues {Curve {keys [
         {arriveTangent=1.7188, interpMode=RCIM_Cubic, leaveTangent=1.7188, tangentMode=RCTM_User, time=0, value=0}
         {arriveTangent=0.0226, interpMode=RCIM_Cubic, leaveTangent=0.0226, tangentMode=RCTM_User, time=0.2, value=0.5}
         {arriveTangent=-0.0031, interpMode=RCIM_Cubic, leaveTangent=-0.0031, tangentMode=RCTM_User, time=0.5, value=0.5}
         {arriveTangent=-1.3404, interpMode=RCIM_Cubic, leaveTangent=-1.3404, tangentMode=RCTM_User, time=1, value=0}
        ]}}}
     Scale Curve=2
    )
    ScaleA=true
    ScaleRGB=true
   )
   &AccelerationForce(Acceleration=<0 0 20>)
   &SolveForcesAndVelocity(Clamp Velocity=false, Limit Acceleration=false)
   &SubUVAnimation(
    End Frame Range Override=30
    "Sprite Renderer "=DataInterface() {propertyValues {SpriteRenderer=&self:smoke_detail_0.NiagaraSpriteRendererProperties_0}}
    Start Frame Range Override=&RandomRangeInt(Minimum=0, Maximum=5)
    UseEndFrame=true
    UseStartFrame=true
   )
  ]
  rendererClasses [&NiagaraSpriteRendererProperties]
  renderers [{rendererClass=&NiagaraSpriteRendererProperties}]
 }
}
summary {
 systemName=NS_Emote_RampCanal_Shadowy
 emitters [
  {emitterName=Smoke001, rendererClasses [&NiagaraMeshRendererProperties]}
  {emitterName=Trail, rendererClasses [&NiagaraMeshRendererProperties]}
  {emitterName=Dark, rendererClasses [&NiagaraSpriteRendererProperties]}
  {emitterName=smoke, rendererClasses [&NiagaraSpriteRendererProperties]}
  {emitterName=smoke_detail, rendererClasses [&NiagaraSpriteRendererProperties]}
 ]
 userVariables [User.bIsFrontEndPreview {type=NiagaraBool}]
}
