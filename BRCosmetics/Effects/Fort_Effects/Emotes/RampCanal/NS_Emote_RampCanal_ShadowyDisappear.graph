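# Niagara system for the RampCanal emote that puffs near-black smoke sprites around a skeletal mesh, as a shadowy disappearance effect.
# Smoke mixes rate and burst spawning on sphere-offset mesh samples, following source movement as it fades in and out.
# Delayed Smoke_Burst throws an expanding, drag-slowed puff using a limited SubUV frame range.
# User.bIsFrontEndPreview lowers the burst's alpha for front-end previews, replacing its randomized scale.
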NiagaraSystem /BRCosmetics/Effects/Fort_Effects/Emotes/RampCanal/NS_Emote_RampCanal_ShadowyDisappear
dependencies {
 usedDataInterfaces [
  &NiagaraDataInterfaceCurve
  &NiagaraDataInterfaceSpriteRendererInfo
  &NiagaraDataInterfaceSkeletalMesh
 ]
 usedDynamicInputs []
 usedModules [
  &self:SystemSpawnScript.SystemScriptSource.SystemScriptGraph.NiagaraNodeAssignment_0.SetVariables_#1
  &SystemState
 ]
}
emitters {
 Smoke {
  emitterUpdateScript [
   &EmitterState(Life Cycle Mode=::NewEnumerator1, Loop Behavior=::NewEnumerator1, Loop Delay=0.5, Loop Duration=0.6, UseLoopDelay=false)
   &SpawnRate(SpawnRate=50, Use Spawn Probability=true)
   &SpawnBurst_Instantaneous(Spawn Count=3, Use Spawn Probability=false, UseLoopCountLimit=false) {module=SpawnBurst_Instantaneous001}
  ]
  particleSpawnScript [
   &InitializeParticle(
    Color=<rgba 0.01 0.01 0.01 1.7>
    Lifetime Max=0.5
    Lifetime Min=0.5
    Lifetime Mode=::NewEnumerator1
    Position Offset=<0 0 30>
    Sprite Rotation Mode=::NewEnumerator2
    Sprite Size Mode=::NewEnumerator1
    Uniform Sprite Size Max=120
    Uniform Sprite Size Min=80
    UsePositionOffset=true
    Write Material Random=true
   )
   &ShapeLocation(
    MaskSpawnGroup=false
    Offset Mode=::NewEnumerator0
    Override Seed=false
    Rotation Angle=1
    Rotation Mode=::NewEnumerator1
    Sphere Radius=40
    UseOwnerScale=true
   )
   &AddVelocity(Velocity=<0 50 0>)
   &SkeletalMeshLocation(
    Bone Sampling Mode=::NewEnumerator4
    MaskSpawnGroup=false
    Sampled Position Offset=&AddVectorToPosition(Position=<10 0 0>, Vector=$Particles.ShapeLocation.ShapeVector)
    Skeletal Mesh=$System.Skeletal Mesh Reader
   )
   &CameraOffset(Camera Offset Amount=10)
  ]
  particleUpdateScript [
   &ParticleState()
   &ScaleSpriteSize(Uniform Curve Sprite Scale=DataInterface() {propertyValues {Curve {keys [
        {arriveTangent=0.0978, interpMode=RCIM_Cubic, leaveTangent=0.0978, tangentMode=RCTM_User, time=0, value=1.2}
        {arriveTangent=-0.0986, tangentMode=RCTM_User, time=1, value=1.4}
       ]}}})
   &SpriteRotationRate(Rotation Rate=&RandomRangeFloat(Minimum=-50, Maximum=50))
   &ScaleColor(
    Scale Alpha=&FloatFromCurve(
     FloatCurve=DataInterface() {propertyValues {Curve {keys [
         <at 0 0 RCIM_Cubic>
         {arriveTangent=0.0034, interpMode=RCIM_Cubic, leaveTangent=0.0034, tangentMode=RCTM_User, time=0.5, value=0.95}
         <at 1 0 RCIM_Cubic>
        ]}}}
     Scale Curve=0.4
    )
    ScaleA=true
    ScaleRGB=true
   )
   &InheritSourceMovement(ApplyForce=false, ApplyPosition=true, ApplyPositionFalloff=false)
   &SolveForcesAndVelocity(Clamp Velocity=false, Limit Acceleration=false)
  ]
  rendererClasses [&NiagaraSpriteRendererProperties]
  renderers [{rendererClass=&NiagaraSpriteRendererProperties}]
 }
 Smoke_Burst {
  emitterUpdateScript [
   &EmitterState(Life Cycle Mode=::NewEnumerator1, Loop Behavior=::NewEnumerator1, Loop Delay=0.45, Loop Duration=1.5, UseLoopDelay=true)
   &SpawnBurst_Instantaneous(Spawn Count=12, Use Spawn Probability=false, UseLoopCountLimit=false)
  ]
  particleSpawnScript [
   &InitializeParticle(
    Color=<rgba 0 0 0 1.7>
    Lifetime Max=0.5
    Lifetime Min=0.3
    Lifetime Mode=::NewEnumerator1
    Sprite Rotation Mode=::NewEnumerator2
    Sprite Size Mode=::NewEnumerator1
    Uniform Sprite Size Max=100
    Uniform Sprite Size Min=70
    UsePositionOffset=false
    Write Material Random=true
   )
   &ShapeLocation(
    MaskSpawnGroup=false
    Offset Mode=::NewEnumerator0
    Override Seed=false
    Rotation Angle=1
    Rotation Mode=::NewEnumerator1
    Sphere Radius=30
    UseOwnerScale=true
   )
   &AddVelocity(Module.Use Falloff Distance=false, Override Seed=false, Velocity Mode=::NewEnumerator1, Velocity Speed=100)
   &AddVelocity(Velocity=<0 30 0>) {module=AddVelocity001}
   &SkeletalMeshLocation(
    Bone Sampling Mode=::NewEnumerator4
    MaskSpawnGroup=false
    Sampled Position Offset=&AddVectorToPosition(Position=<0 0 0>, Vector=$Particles.ShapeLocation.ShapeVector)
    Skeletal Mesh=$System.Skeletal Mesh Reader
   )
   &CameraOffset(Camera Offset Amount=10)
   &SubUVAnimation(
    End Frame Range Override=3
    "Sprite Renderer "=DataInterface() {propertyValues {SpriteRenderer=&self:Smoke_Burst_0.NiagaraSpriteRendererProperties_1}}
    Start Frame Range Override=2
    SubUV Animation Mode:ENiagara_SubUVLookupModeV2=::NewEnumerator2
    UseEndFrame=true
    UseStartFrame=true
   )
  ]
  particleUpdateScript [
   &ParticleState()
   &ScaleSpriteSize(Uniform Curve Sprite Scale=DataInterface() {propertyValues {Curve {keys [
        {arriveTangent=3.1869, interpMode=RCIM_Cubic, leaveTangent=3.1869, tangentMode=RCTM_User, time=0, value=0.5}
        {arriveTangent=0.5887, interpMode=RCIM_Cubic, leaveTangent=0.5887, tangentMode=RCTM_User, time=0.2, value=0.95}
        {arriveTangent=0.082, interpMode=RCIM_Cubic, leaveTangent=0.082, tangentMode=RCTM_User, time=1, value=1.2}
       ]}}})
   &SpriteRotationRate(Rotation Rate=&RandomRangeFloat(Minimum=5, Maximum=50))
   &ScaleColor(
    Scale Alpha=&FloatFromCurve(
     FloatCurve=DataInterface() {propertyValues {Curve {keys [
         <at 0 0 RCIM_Cubic>
         {arriveTangent=0.0034, interpMode=RCIM_Cubic, leaveTangent=0.0034, tangentMode=RCTM_User, time=0.1, value=1}
         {arriveTangent=-0.0216, interpMode=RCIM_Cubic, leaveTangent=-0.9755, tangentMode=RCTM_Break, time=0.3, value=1}
         <at 1 0 RCIM_Cubic>
        ]}}}
     Scale Curve=&MakeCustomFloatFromBool(Bool=$User.bIsFrontEndPreview, True Float=0.8, False Float=&RandomRangeFloat(Minimum=1.2, Maximum=1.7))
    )
    ScaleA=true
    ScaleRGB=true
   )
   &Drag(Use Linear Drag=true, Use Rotational Drag=true)
   &SolveForcesAndVelocity(Clamp Velocity=false, Limit Acceleration=false)
  ]
  rendererClasses [&NiagaraSpriteRendererProperties]
  renderers [{rendererClass=&NiagaraSpriteRendererProperties}]
 }
}
summary {
 systemName=NS_Emote_RampCanal_ShadowyDisappear
 emitters [
  {emitterName=Smoke, rendererClasses [&NiagaraSpriteRendererProperties]}
  {emitterName=Smoke_Burst, rendererClasses [&NiagaraSpriteRendererProperties]}
 ]
 userVariables [User.bIsFrontEndPreview {type=NiagaraBool}]
}
