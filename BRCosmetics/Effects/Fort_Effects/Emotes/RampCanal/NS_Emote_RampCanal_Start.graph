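# Niagara system for the start of the RampCanal emote: a blue burst of mesh trails, ground decal, flare and debris.
# Mesh emitters layer center, flame and ground trails with a wind cone, placing some via per-partition offset arrays and animating material parameters over life.
# GPU sprite debris rises with curl noise, its spawning cut off by emitter age.
# Time-of-day scalars dim flare and ground trail; a front-end preview flag boosts wind cone alpha.
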NiagaraSystem /BRCosmetics/Effects/Fort_Effects/Emotes/RampCanal/NS_Emote_RampCanal_Start
dependencies {
 usedDataInterfaces [
  &NiagaraDataInterfaceCurve
  &NiagaraDataInterfaceSpriteRendererInfo
  &NiagaraDataInterfaceArrayFloat3
  &NiagaraDataInterfaceColorCurve
 ]
 usedDynamicInputs []
 usedModules [&SystemState]
}
emitters {
 CenterTrail {
  emitterUpdateScript [
   &EmitterState(Inactive Response=::NewEnumerator1, Life Cycle Mode=::NewEnumerator1, Loop Behavior=::NewEnumerator1, UseLoopDelay=false)
   &SpawnBurst_Instantaneous(Spawn Count=3, Use Spawn Probability=false, UseLoopCountLimit=false)
  ]
  particleSpawnScript [
   &PartitionParticles(Number Of Partitions=3)
   &InitializeParticle(
    Lifetime=4
    Mesh Scale=&SelectVectorFromArray(
     Direct Array Index=$Particles.PartitionParticles.Partition
     Vector Selection Array=DataInterface() {propertyValues {FloatData [<2.5 2.5 1.1>, <2.5 2.5 1.3>, <2.5 2.5 1.1>]}}
    )
    Mesh Scale Mode=::NewEnumerator4
    Position Offset=&SelectVectorFromArray(
     Direct Array Index=$Particles.PartitionParticles.Partition
     Vector Selection Array=DataInterface() {propertyValues {FloatData [<35 15 -20>, <0 0 -10>, <-35 -15 0>]}}
    )
    UsePositionOffset=true
    Write Material Random=true
   )
   &CameraOffset(Camera Offset Amount=-50)
   &InitialMeshOrientation(
    Mesh Orientation Mode=::NewEnumerator2
    Rotation=&RandomRangeVector(Minimum=<0 0 0>, Maximum=<0 0 1>)
    Use Rotation Vector=true
   )
  ]
  particleUpdateScript [
   &ParticleState()
   &ScaleMeshSize(Scale Factor=&MakeVector(
     X=1
     Y=1
     Z=&FloatFromCurve(FloatCurve=DataInterface() {propertyValues {Curve {keys [{arriveTangent=0.7581, interpMode=RCIM_Cubic, leaveTangent=0.7581, tangentMode=RCTM_User, time=0, value=0.6}, <at 1 1 RCIM_Cubic>]}}})
    ))
   &ScaleMeshSize(Scale Factor=&Multiply_VectorByFloat(Vector=<1 1 1>, Float=0.8)) {module=ScaleMeshSize001}
   &ScaleColor(
    Scale Alpha=&FloatFromCurve(FloatCurve=DataInterface() {propertyValues {Curve {keys [<at 0 0>,<at 0.2 1>,<at 0.6 1>,<at 1 0>]}}})
    Scale RGB=&MakeVectorFromLinearColorRGB(Color=&ColorFromCurve(ColorCurve=DataInterface() {propertyValues {
        AlphaCurve {keys []}
        BlueCurve {keys [<at 0.1302 7 RCIM_Cubic>]}
        GreenCurve {keys [<at 0.1302 2.9978 RCIM_Cubic>]}
        RedCurve {keys [<at 0.1302 0 RCIM_Cubic>]}
       }}))
    ScaleA=true
    ScaleRGB=true
   )
   &DynamicMaterialParameters(
    Index 0 Param 1=&RandomRangeFloat(Minimum=0.4, Maximum=0.8)
    Index 0 Param 2=&RandomRangeFloat(Minimum=0.4, Maximum=0.8)
    Index 0 Param 3=&RandomRangeFloat(Minimum=20, Maximum=10)
    Index 0 Param 4=1
    Index 3 Param 1=&FloatFromCurve(FloatCurve=DataInterface() {propertyValues {Curve {keys [
         <at 0 -625.0001 RCIM_Cubic>
         {arriveTangent=712.5001, interpMode=RCIM_Cubic, leaveTangent=712.5001, time=1, value=0}
         {arriveTangent=1495.8616, interpMode=RCIM_Cubic, leaveTangent=1495.8616, tangentMode=RCTM_User, time=2, value=800}
        ]}}})
    Module.Index0Float0Write=true
    Module.Index0Float1Write=true
    Module.Index0Float2Write=true
    Module.Index0Float3Write=true
    Module.Index1Float0Write=true
    Module.Index1Float1Write=true
    Module.Index1Float2Write=true
    Module.Index1Float3Write=true
    Module.Index3Float0Write=true
    Module.Index3Float1Write=true
    Module.Index3Float2Write=true
    Module.Index3Float3Write=true
    Write Parameter Index 1=::NewEnumerator1
    Write Parameter Index 3=::NewEnumerator1
   )
  ]
  rendererClasses [&NiagaraMeshRendererProperties]
  renderers [{rendererClass=&NiagaraMeshRendererProperties, rendererIndex=0}]
 }
 Debris {
  simTarget=GPUComputeSim
  emitterUpdateScript [
   &EmitterState(Inactive Response=::NewEnumerator1, Life Cycle Mode=::NewEnumerator1, Loop Behavior=::NewEnumerator1, UseLoopDelay=true)
   &SpawnRate(
    Spawn Probability=&FloatFromCurve(FloatCurve=DataInterface() {propertyValues {Curve {keys [<at 0 1>,<at 2.9 1>,<at 3.2 0>]}}}, CurveIndex=$Emitter.Age)
    SpawnRate=30
    Use Spawn Probability=true
   )
  ]
  particleSpawnScript [
   &InitializeParticle(
    Color=<rgba 0 0.6423 1.5 1>
    Lifetime Max=1
    Lifetime Mode=::NewEnumerator1
    Mass Mode=::NewEnumerator1
    Sprite Rotation Mode=::NewEnumerator2
    Sprite Size Max=<50 50>
    Sprite Size Min=<20 20>
    Sprite Size Mode=::NewEnumerator2
    Sprite UV Mode=::NewEnumerator1
    UsePositionOffset=false
    Write Material Random=true
   )
   &ShapeLocation(
    Handle Radius=0
    Large Radius=80
    MaskSpawnGroup=false
    Offset=<0 0 20>
    Offset Mode=::NewEnumerator0
    Override Seed=false
    Shape Primitive=::NewEnumerator3
    UseOwnerScale=true
   )
   &AddVelocity(Module.Use Falloff Distance=false, Origin Offset=<0 0 -300>, Override Seed=false, Velocity Mode=::NewEnumerator1, Velocity Speed=200)
   &AddVelocity() {module=AddVelocity001}
   &DynamicMaterialParameters(
    Index 0 Param 1=3
    Module.Index0Float0Write=true
    Module.Index0Float1Write=false
    Module.Index0Float2Write=false
    Module.Index0Float3Write=false
   )
  ]
  particleUpdateScript [
   &ParticleState()
   &ScaleSpriteSize(
    Uniform Curve Scale=0.2
    Uniform Curve Sprite Scale=DataInterface() {propertyValues {Curve {keys [{tangentMode=RCTM_User, time=0, value=1}, {tangentMode=RCTM_User, time=1, value=0.9}]}}}
   ) {module=ScaleSpriteSize001}
   &CurlNoiseForce(Bias Noise=false, Noise Frequency=20, Noise Strength=1000, Pan Noise=false)
   &GravityForce(Gravity=&Multiply_VectorByFloat(Vector=<0 0 200>, Float=1))
   &Drag(
    Drag=&FloatFromCurve(
     FloatCurve=DataInterface() {propertyValues {Curve {defaultValue=1, keys [{tangentMode=RCTM_User, time=1.5, value=5}, {tangentMode=RCTM_User, time=1.7, value=1.0777}]}}}
     CurveIndex=$Emitter.Age
    )
    Use Linear Drag=true
    Use Rotational Drag=true
   )
   &SolveForcesAndVelocity(Clamp Velocity=false, Limit Acceleration=false)
   &ScaleColor(
    Scale Alpha=&FloatFromCurve(FloatCurve=DataInterface() {propertyValues {Curve {keys [<at 0 0>,<at 0.2 1>,<at 0.8 1>,<at 1 0>]}}})
    ScaleA=true
    ScaleRGB=true
   )
   &SubUVAnimation(
    End Frame Range Override=16
    "Sprite Renderer "=DataInterface() {propertyValues {SpriteRenderer=&self:Debris_0.NiagaraSpriteRendererProperties_0}}
    SubUV Animation Mode:ENiagara_SubUVLookupModeV2=::NewEnumerator2
    UseEndFrame=true
    UseStartFrame=true
   )
   &FN_Sprite_Attributes(
    Color=&MakeLinearColorFromVectorAndFloat(
     "Vector (RGB)"=&MakeVectorFromLinearColorRGB(Color=<rgba 0 2.0285 5 1>)
     "Float (Alpha)"=&FloatFromCurve(
      FloatCurve=DataInterface() {propertyValues {Curve {keys [
          <at 0 0 RCIM_Cubic>
          {arriveTangent=-0.0291, interpMode=RCIM_Cubic, leaveTangent=-0.0291, tangentMode=RCTM_User, time=0.2, value=1}
          {arriveTangent=-0.05, interpMode=RCIM_Cubic, leaveTangent=-3.3737, tangentMode=RCTM_Break, time=0.3, value=1}
          <at 1 0 RCIM_Cubic>
         ]}}}
      Scale Curve=0.02
     )
    )
    Create Color Attribute=true
    Create SpriteSize Attribute=true
    SpriteSize=&Multiply_Vector2D_ByFloat(Float=40, Vector2D=<1 1>)
    Write to Module Namespace=true
   )
   /SetVariables() {isSetParameters=true}
  ]
  rendererClasses [&NiagaraSpriteRendererProperties]
  renderers [
   {rendererClass=&NiagaraSpriteRendererProperties}
   {rendererClass=&NiagaraSpriteRendererProperties, rendererIndex=1}
  ]
 }
 Decal {
  emitterUpdateScript [
   &EmitterState(Inactive Response=::NewEnumerator1, Life Cycle Mode=::NewEnumerator1, Loop Behavior=::NewEnumerator1, UseLoopDelay=false)
   &SpawnBurst_Instantaneous(Use Spawn Probability=false, UseLoopCountLimit=false)
  ]
  particleSpawnScript [
   &InitializeParticle(Color=<rgba 0 0 0 1>, Lifetime=3.6, UsePositionOffset=true, Write Material Random=true)
   /Niagara/Modules/Spawn/Decal/InitialDecalOrientation(
    Orientation Vector:Vector3f=<0 0 1>
    Rotation:Vector3f=&RandomRangeVector(Minimum=<0 0 0>, Maximum=<0 0 0>)
    Use Rotation Vector=true
   )
  ]
  particleUpdateScript [
   &ParticleState(Kill Particles When Lifetime Has Elapsed=false, Let Infinitely Lived Particles Die When Emitter Deactivates=true)
   /Niagara/Modules/Decal/Decal_Attributes(
    Color=1
    Color Mode=0
    Decal Color:LinearColor=<rgba 0.0032 0.0061 0.01 2>
    Decal Fade:NiagaraFloat=&FloatFromCurve(FloatCurve=DataInterface() {propertyValues {Curve {keys [<at 0 0>, {arriveTangent=1.25, leaveTangent=1.25, time=0.2, value=1}, {arriveTangent=-1.25, leaveTangent=-1.25, time=0.8, value=1}, <at 1 0>]}}})
    Decal Size:Vector3f=&Multiply_VectorByFloat(Vector=<5 5 5>, Float=100)
    Fade=1
    Position=1
    Size=1
    UsePositionOffset=true
   )
  ]
  rendererClasses [&NiagaraDecalRendererProperties]
  renderers [{rendererClass=&NiagaraDecalRendererProperties, rendererIndex=0}]
 }
 FlameTrailLong001 {
  emitterUpdateScript [
   &EmitterState(Inactive Response=::NewEnumerator1, Life Cycle Mode=::NewEnumerator1, Loop Behavior=::NewEnumerator1, Loop Duration=3, UseLoopDelay=true)
   &SpawnBurst_Instantaneous(Spawn Count=6, Use Spawn Probability=false, UseLoopCountLimit=false)
  ]
  particleSpawnScript [
   &PartitionParticles(Number Of Partitions=6)
   &InitializeParticle(
    Lifetime=2.5
    Mesh Scale Max=<1.3 1.3 0.8>
    Mesh Scale Min=<1 1 0.7>
    Mesh Scale Mode=::NewEnumerator2
    Position Offset=&SelectVectorFromArray(
     Direct Array Index=$Particles.PartitionParticles.Partition
     Vector Selection Array=DataInterface() {propertyValues {FloatData [<-25 30 -20>, <-15 20 0>, <-20 20 -20>, <10 0 0>, <10 0 0>, <10 0 -20>]}}
    )
    UsePositionOffset=true
    Write Material Random=true
   )
   &InitialMeshOrientation(Mesh Orientation Mode=::NewEnumerator7, Rotation=<0 0.05 0>, Use Rotation Vector=true) {module=InitialMeshOrientation001}
   &InitialMeshOrientation(
    Mesh Orientation Mode=::NewEnumerator2
    Rotation=&SelectVectorFromArray(
     Direct Array Index=$Particles.PartitionParticles.Partition
     Vector Selection Array=DataInterface() {propertyValues {FloatData [<0 0 -0.05>, <0 0 0.05>, <0 0 0.2>, <0 0 0.35>, <0 0 0.5>, <0 0 0.6>]}}
    )
    Use Rotation Vector=true
   )
  ]
  particleUpdateScript [
   &ParticleState()
   &ScaleMeshSize(Scale Factor=&Multiply_VectorByFloat(
     Vector=<1 1 1>
     Float=&FloatFromCurve(FloatCurve=DataInterface() {propertyValues {Curve {keys [<at 0 0.4>,<at 0.1 0.95>,<at 1 1>]}}}, Scale Curve=1.9)
    ))
   &DynamicMaterialParameters(
    Index 0 Param 1=&RandomRangeFloat(Minimum=3.5, Maximum=4.5)
    Index 0 Param 2=&RandomRangeFloat(Minimum=0, Maximum=1)
    Index 0 Param 3=&RandomRangeFloat(Minimum=0.2, Maximum=0.2)
    Index 0 Param 4=&FloatFromCurve(FloatCurve=DataInterface() {propertyValues {Curve {keys [
         <at 0 0.2 RCIM_Cubic>
         {arriveTangent=0.7131, interpMode=RCIM_Cubic, leaveTangent=0.7131, tangentMode=RCTM_User, time=0.6, value=0.55}
         <at 1 0.8>
        ]}}})
    Index 1 Param 1=&FloatFromCurve(
     FloatCurve=DataInterface() {propertyValues {Curve {keys [<at 0 0>,<at 1 1>]}}}
     Scale Curve=&FloatFromCurve(FloatCurve=DataInterface() {propertyValues {Curve {keys [<at 0 3>,<at 1 1.5>]}}})
    )
    Index 1 Param 2=&RandomRangeFloat(Minimum=0, Maximum=1)
    Module.Index0Float0Write=true
    Module.Index0Float1Write=true
    Module.Index0Float2Write=true
    Module.Index0Float3Write=true
    Module.Index1Float0Write=true
    Module.Index1Float1Write=true
    Module.Index1Float2Write=false
    Module.Index1Float3Write=false
    Write Parameter Index 1=::NewEnumerator1
   )
   &ScaleColor(
    Scale Alpha=&FloatFromCurve(FloatCurve=DataInterface() {propertyValues {Curve {keys [<at 0 1>,<at 0.5 0.95>,<at 1 0>]}}})
    Scale RGB=&Multiply_VectorByFloat(
     Vector=&MakeVectorFromLinearColorRGB(Color=<rgba 0 0.2482 0.599 1>)
     Float=&FloatFromCurve(FloatCurve=DataInterface() {propertyValues {Curve {keys [<at 0 0.03>,<at 0.2 0.01>,<at 1 0.01>]}}})
    )
    ScaleA=true
    ScaleRGB=true
   )
   /SetVariables(Particles.Color002=&MakeLinearColorFromVectorAndFloat(
     "Vector (RGB)"=&MakeVectorFromLinearColorRGB(Color=<rgba 0.0745 0.3216 0.7448 1>)
     "Float (Alpha)"=&FloatFromCurve(FloatCurve=DataInterface() {propertyValues {Curve {keys [<at 0.4 1>,<at 0.7 0>]}}}, Scale Curve=0.5)
    )) {isSetParameters=true}
   &SolveForcesAndVelocity(Clamp Velocity=false, Limit Acceleration=false)
  ]
  rendererClasses [&NiagaraMeshRendererProperties]
  renderers [
   {rendererClass=&NiagaraMeshRendererProperties, rendererIndex=0}
   {rendererClass=&NiagaraMeshRendererProperties, rendererIndex=1}
  ]
 }
 Flare002 {
  emitterUpdateScript [
   &EmitterState(Life Cycle Mode=::NewEnumerator1, Loop Behavior=::NewEnumerator1, Loop Duration=1, UseLoopDelay=false)
   &SpawnBurst_Instantaneous(Spawn Count=1, Use Spawn Probability=false)
  ]
  particleSpawnScript [
   &InitializeParticle(
    Color=<rgba 0.0164 0.3964 2 1>
    Lifetime=0.65
    Sprite Size Mode=::NewEnumerator3
    Uniform Sprite Size=650
    UsePositionOffset=false
    Write Material Random=true
   )
   &DynamicMaterialParameters(
    Index 0 Param 1=&RandomRangeFloat(Minimum=0.4, Maximum=0.6)
    Index 0 Param 2=1
    Index 0 Param 3=1
    Index 0 Param 4=1
    Module.Param0WriteEnabled=true
    Module.Param1WriteEnabled=true
    Module.Param2WriteEnabled=true
    Module.Param3WriteEnabled=true
   )
   &SystemLocation(MaskSpawnGroup=false, Offset=<0 0 30>)
   &CameraOffset(Camera Offset Amount=50)
   &SpriteFacingAndAlignment(
    Alignment Coordinate Space=::Local
    Facing Coordinate Space=::Local
    Sprite Alignment=<0 0 1>
    Sprite Facing=<0 0 1>
    WriteAlignment=true
    WriteFacing=true
   )
  ]
  particleUpdateScript [
   &ParticleState()
   &SolveForcesAndVelocity(Clamp Velocity=false, Limit Acceleration=false)
   &ScaleColor(
    Scale Alpha=&Multiply_Float(
     A=1
     B=&FloatFromCurve(FloatCurve=DataInterface() {propertyValues {Curve {keys [<at 0 0>,<at 0.15 1>,<at 0.4 0.4>,<at 1 0>]}}})
    )
    ScaleA=true
    ScaleRGB=true
   )
   &FN_Sprite_Attributes(
    Color=&MakeLinearColorFromVectorAndFloat(
     "Vector (RGB)"=&MakeVectorFromLinearColorRGB(Color=<rgba 0.0303 1.0548 3 1>)
     "Float (Alpha)"=&FloatFromCurve(
      FloatCurve=DataInterface() {propertyValues {Curve {keys [
          {interpMode=RCIM_Cubic, tangentMode=RCTM_User, time=0, value=0}
          <at 0.1 1 RCIM_Cubic>
          {arriveTangent=-2.4416, interpMode=RCIM_Cubic, leaveTangent=-2.4416, tangentMode=RCTM_User, time=0.25, value=0}
         ]}}}
      Scale Curve=&Multiply_Float(
       A=0.5
       B=&FNTimeOfDayScalar(Night Scale=0.5, Morning Scale=0.7, Day Scale=0.7, Evening Scale=0.7)
      )
     )
    )
    Create Color Attribute=true
    Create SpriteSize Attribute=true
    SpriteSize=&Multiply_Vector2D_ByFloat(Float=1, Vector2D=<200 200>)
    Write to Module Namespace=true
   )
  ]
  rendererClasses [&NiagaraSpriteRendererProperties]
  renderers [
   {rendererClass=&NiagaraSpriteRendererProperties}
   {rendererClass=&NiagaraSpriteRendererProperties, rendererIndex=1}
  ]
 }
 GroundTrail {
  emitterUpdateScript [
   &EmitterState(Inactive Response=::NewEnumerator1, Life Cycle Mode=::NewEnumerator1, Loop Behavior=::NewEnumerator1, UseLoopDelay=false)
   &SpawnBurst_Instantaneous(Spawn Count=5, Use Spawn Probability=false, UseLoopCountLimit=false)
   &SpawnRate(
    Spawn Probability=&FloatFromCurve(FloatCurve=DataInterface() {propertyValues {Curve {keys [<at 0 1>,<at 2 1>,<at 2.5 0>]}}}, CurveIndex=$Emitter.Age)
    SpawnRate=10
    Use Spawn Probability=true
   )
  ]
  particleSpawnScript [
   &InitializeParticle(
    Color Mode=::NewEnumerator0
    Lifetime Max=2.5
    Lifetime Min=1.5
    Lifetime Mode=::NewEnumerator1
    Mesh Scale Max=<4 4 3>
    Mesh Scale Min=<4 4 1.5>
    Mesh Scale Mode=::NewEnumerator2
    Position Offset=<0 0 -20>
    UsePositionOffset=true
    Write Material Random=true
   )
   &InitialMeshOrientation(
    Mesh Orientation Mode=::NewEnumerator2
    Rotation=&RandomRangeVector(Minimum=<0 0 0>, Maximum=<0 0 1>)
    Use Rotation Vector=true
   )
   &ShapeLocation(
    Handle Radius=10
    Large Radius=80
    MaskSpawnGroup=false
    Override Seed=false
    Shape Primitive=::NewEnumerator3
    Surface Distribution=1
    UseOwnerScale=true
   )
   &AddVelocity(Velocity=<0 0 30>)
   /SetVariables(Particles.VisibilityTag=&RandomRangeInt(Minimum=0, Maximum=1)) {isSetParameters=true}
  ]
  particleUpdateScript [
   &ParticleState()
   &ScaleMeshSize(Scale Factor=&MakeVector(
     X=1
     Y=1
     Z=&FloatFromCurve(FloatCurve=DataInterface() {propertyValues {Curve {keys [<at 0 1>,<at 1 1.5>]}}})
    ))
   &UpdateMeshOrientation(Rotation Rate=&RandomRangeFloat(Minimum=0, Maximum=0.1), Rotation Vector=<0 0 1>)
   &DynamicMaterialParameters(
    Index 0 Param 1=&FloatFromCurve(
     FloatCurve=DataInterface() {propertyValues {Curve {keys [
         {arriveTangent=2.5259, interpMode=RCIM_Cubic, leaveTangent=2.5259, tangentMode=RCTM_User, time=0, value=-1}
         {arriveTangent=1.2664, interpMode=RCIM_Cubic, leaveTangent=1.2664, tangentMode=RCTM_User, time=1, value=1}
        ]}}}
     Scale Curve=&RandomRangeFloat(Minimum=0.8, Maximum=1.3)
    )
    Index 0 Param 2=&RandomRangeFloat(Minimum=0.6, Maximum=1.5)
    Index 0 Param 3=&RandomRangeFloat(Minimum=0.1, Maximum=0.6)
    Module.Index0Float0Write=true
    Module.Index0Float1Write=true
    Module.Index0Float2Write=true
    Module.Index0Float3Write=false
   )
   &ScaleColor(
    Scale Alpha=&FloatFromCurve(
     FloatCurve=DataInterface() {propertyValues {Curve {keys [
         <at 0 0 RCIM_Cubic>
         {arriveTangent=-0.0065, interpMode=RCIM_Cubic, leaveTangent=-0.0065, tangentMode=RCTM_User, time=0.1, value=1}
         {arriveTangent=-2.9294, interpMode=RCIM_Cubic, leaveTangent=-2.9294, tangentMode=RCTM_User, time=0.4483, value=0.4427}
         <at 1 0 RCIM_Cubic>
        ]}}}
     Scale Curve=&RandomRangeFloat(Minimum=1, Maximum=1)
    )
    Scale RGB=&Multiply_VectorByFloat(
     Vector=&MakeVectorFromLinearColorRGB(Color=<rgba 0 0.0021 0.01 1>)
     Float=&Multiply_Float(
      A=&RandomRangeFloat(Minimum=0.3, Maximum=0.7)
      B=&FNTimeOfDayScalar(Night Scale=0.6, Morning Scale=0.8, Evening Scale=0.7)
     )
    )
    ScaleA=true
    ScaleRGB=true
   )
   /SetVariables(Particles.Color002=&MakeLinearColorFromVectorAndFloat(
     "Vector (RGB)"=&MakeVectorFromLinearColorRGB(Color=<rgba 0.0208 0.1882 0.4167 1>)
     "Float (Alpha)"=&FloatFromCurve(
      FloatCurve=DataInterface() {propertyValues {Curve {keys [
          <at 0 0 RCIM_Cubic>
          {arriveTangent=-0.0065, interpMode=RCIM_Cubic, leaveTangent=-0.0065, tangentMode=RCTM_User, time=0.2, value=1}
          {arriveTangent=-2.3244, interpMode=RCIM_Cubic, leaveTangent=-2.3244, tangentMode=RCTM_User, time=0.5, value=0.4}
          <at 1 0 RCIM_Cubic>
         ]}}}
      Scale Curve=&RandomRangeFloat(Minimum=1, Maximum=1.3)
     )
    )) {isSetParameters=true}
   &SolveForcesAndVelocity(Clamp Velocity=false, Limit Acceleration=false)
  ]
  rendererClasses [&NiagaraMeshRendererProperties]
  renderers [
   {rendererClass=&NiagaraMeshRendererProperties, rendererIndex=0}
   {rendererClass=&NiagaraMeshRendererProperties, rendererIndex=1}
  ]
 }
 WindCone_Burst001 {
  emitterUpdateScript [
   &EmitterState(Inactive Response=::NewEnumerator1, Life Cycle Mode=::NewEnumerator1, Loop Behavior=::NewEnumerator1, UseLoopDelay=false)
   &SpawnBurst_Instantaneous(Use Spawn Probability=false, UseLoopCountLimit=false)
  ]
  particleSpawnScript [
   &InitializeParticle(
    Lifetime=3.8
    Mesh Scale=<0.1 1 1>
    Mesh Scale Mode=::NewEnumerator4
    Position Offset=<0 0 5>
    Sprite Size Mode=::NewEnumerator3
    Uniform Sprite Size=300
    UsePositionOffset=true
    Write Material Random=true
   )
   &InitialMeshOrientation(Mesh Orientation Mode=::NewEnumerator2, Rotation=<0 0.25 0>, Use Rotation Vector=true)
   /SetVariables(Particles.Mesh Orientation002:Quat4f=<0 0 0 0>) {isSetParameters=true}
   /SetVariables(Particles.Dyn002:Vector4f=&MakeVector4(X=1, Y=1, Z=0, W=0)) {isSetParameters=true}
   &SpriteFacingAndAlignment(Sprite Facing=<0 0 1>, WriteAlignment=true, WriteFacing=true)
  ]
  particleUpdateScript [
   &ParticleState()
   &ScaleMeshSize(Scale Factor=&VectorFromFloat(Value=&FloatFromCurve(
      FloatCurve=DataInterface() {propertyValues {Curve {keys [
          {arriveTangent=4.2179, interpMode=RCIM_Cubic, leaveTangent=4.2179, tangentMode=RCTM_User, time=0, value=0}
          {arriveTangent=0.0013, interpMode=RCIM_Cubic, leaveTangent=0.0013, tangentMode=RCTM_User, time=0.1, value=1}
          <at 1 1 RCIM_Cubic>
         ]}}}
      Scale Curve=3
     )))
   &DynamicMaterialParameters(
    Index 0 Param 1=1
    Index 0 Param 4=1
    Module.Index0Param0Write=true
    Module.Index0Param1Write=true
    Module.Index0Param2Write=true
    Module.Index0Param3Write=true
   )
   &ScaleColor(
    Scale Alpha=&FloatFromCurve(
     FloatCurve=DataInterface() {propertyValues {Curve {keys [
         <at 0 0>
         {arriveTangent=-1.2369, leaveTangent=-1.2369, time=0.01, value=1}
         {arriveTangent=-1.1111, leaveTangent=-1.1111, time=0.9, value=1}
         <at 1 0>
        ]}}}
     Scale Curve=1.3
    )
    Scale RGB=&MakeVectorFromLinearColorRGB(Color=<rgba 0 0 0 1>)
    ScaleA=true
    ScaleRGB=true
   )
   /SetVariables(Particles.Color003:LinearColor=&MakeLinearColorFromVectorAndFloat(
     "Vector (RGB)"=&MakeVectorFromLinearColorRGB(Color=<rgba 0 0.0027 0.01 1>)
     "Float (Alpha)"=&FloatFromCurve(
      FloatCurve=DataInterface() {propertyValues {Curve {keys [
          <at 0 0>
          {arriveTangent=-1.2369, leaveTangent=-1.2369, time=0.1, value=1}
          {arriveTangent=-1.1111, leaveTangent=-1.1111, time=0.9, value=1}
          <at 1 0>
         ]}}}
      Scale Curve=&MakeCustomFloatFromBool(Bool=$User.bIsFrontEndPreview, True Float=1.5, False Float=1.1)
     )
    )) {isSetParameters=true}
   /SetVariables(Particles.Wave_Color:LinearColor=&MakeLinearColorFromVectorAndFloat(
     "Vector (RGB)"=&MakeVectorFromLinearColorRGB(Color=<rgba 0 0.032 0.08 1>)
     "Float (Alpha)"=&FloatFromCurve(
      FloatCurve=DataInterface() {propertyValues {Curve {keys [
          <at 0 0>
          {arriveTangent=-1.2369, leaveTangent=-1.2369, time=0.1, value=1}
          {arriveTangent=-1.1111, leaveTangent=-1.1111, time=0.8, value=1}
          <at 1 0>
         ]}}}
      Scale Curve=5
     )
    )) {isSetParameters=true}
   /SetVariables(Particles.Wave_Size:Vector2f=&Vector2DFromFloat(Value=&Multiply_Float(
      A=260
      B=&FloatFromCurve(FloatCurve=DataInterface() {propertyValues {Curve {keys [
           {arriveTangent=0.82, interpMode=RCIM_Cubic, leaveTangent=0.82, tangentMode=RCTM_User, time=0, value=0.5}
           {arriveTangent=0.0101, interpMode=RCIM_Cubic, leaveTangent=0.0101, tangentMode=RCTM_User, time=0.2, value=1}
           {arriveTangent=0.004, interpMode=RCIM_Cubic, leaveTangent=0.004, tangentMode=RCTM_User, time=1, value=1}
          ]}}})
     ))) {isSetParameters=true}
  ]
  rendererClasses [&NiagaraMeshRendererProperties, &NiagaraSpriteRendererProperties]
  renderers [
   {rendererClass=&NiagaraMeshRendererProperties, rendererIndex=0}
   {rendererClass=&NiagaraSpriteRendererProperties, rendererIndex=1}
   {rendererClass=&NiagaraSpriteRendererProperties, rendererIndex=2}
  ]
 }
}
summary {
 systemName=NS_Emote_RampCanal_Start
 emitters [
  {emitterName=Debris, simTarget=GPUComputeSim, rendererClasses [&NiagaraSpriteRendererProperties]}
  {emitterName=Decal, rendererClasses [&NiagaraDecalRendererProperties]}
  {
   emitterName=WindCone_Burst001
   rendererClasses [&NiagaraMeshRendererProperties, &NiagaraSpriteRendererProperties]
  }
  {emitterName=FlameTrailLong001, rendererClasses [&NiagaraMeshRendererProperties]}
  {emitterName=Flare002, rendererClasses [&NiagaraSpriteRendererProperties]}
  {emitterName=GroundTrail, rendererClasses [&NiagaraMeshRendererProperties]}
  {emitterName=CenterTrail, rendererClasses [&NiagaraMeshRendererProperties]}
 ]
 userVariables [
  User.bIsFrontEnd {type=NiagaraBool}
  User.bIsFrontEndPreview {type=NiagaraBool}
  User.IsFrontEnd {type=NiagaraBool}
 ]
}
